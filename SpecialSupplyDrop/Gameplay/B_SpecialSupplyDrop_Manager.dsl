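; Server-side Blueprint manager that spawns special supply drops as storm phases progress, when enabled for the playlist.
; At startup it runs one EQS query for candidate locations, then trims the cached list to the safe zone on each new storm circle.
; Game phase events trigger delayed spawn attempts, gated by allowed circles, per-circle chance and drop caps.
; Chance rolls once per phase or per attempt; debug events can force drops or phases by ID.
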(event Custom|Initialize
  ;[1 1 1 1 14] This will likely be an expensive query, so we do it once at the start before players are on the island, cache the locations, and filter them by safe zone later. This query does not need to include safe zone filtering because of this, but should handle things like excluding water, walkable surfaces, etc.
  (Query|DoEQSQueryforSpawnLocations)
  (Development|PrintString "Emerald Glass Supply Drop Manager starting up")
  ;[1 1 1] Bind listener to game phase step changes to determine when to try to spawn supply drops
  (bind _returnvalue (GameplayEventRouter|GetEventRouter "/Script/FortniteGame.FortGameState" self))
  (bind (_asynctaskproxy _payload) (GameplayEventRouter|StartListeningToStatefulEvent _returnvalue "/Script/FortniteGame.FortGamePhaseStepUpdatedEvent")
    (:then
      (Variables|Default|SetGamePhaseStepUpdatedAsyncTask _asynctaskproxy)
      (bind (_asynctaskproxy_2 _payload_2) (GameplayEventRouter|StartListeningToStatefulEvent _returnvalue "/Script/FortniteGame.FortCreativeStormShieldGamePhaseStepUpdatedEvent")
        (:then
          (Variables|Default|SetCreativeStormShieldGamePhaseStepUpdatedAsyncTask _asynctaskproxy_2))
        (:OnEventReceived
          _)
        (:OnSavedEventStateExists
          _))
      (CallFunction|OnGamePhaseStepChanged (.GamePhaseStep (Utilities|Struct|BreakFortCreativeStormShieldGamePhaseStepUpdatedEvent _payload_2))))
    (:OnEventReceived
      _)
    (:OnSavedEventStateExists
      _))
  (CallFunction|OnGamePhaseStepChanged (Utilities|Struct|BreakFortGamePhaseStepUpdatedEvent _payload)))

(event EventEndPlay
  (params (EndPlayReason :type EEndPlayReason%20Enum :default "Destroyed"))
  (bind _gamephasestepupdatedasynctask (Variables|Default|GetGamePhaseStepUpdatedAsyncTask)
    (:then
      (GameplayEventRouter|StopListeningtoStatefulEvent _gamephasestepupdatedasynctask))
    (:else
      _))
  (bind _creativestormshieldgamephasestepupdatedasynctask (Variables|Default|GetCreativeStormShieldGamePhaseStepUpdatedAsyncTask)
    (:then
      (GameplayEventRouter|StopListeningtoStatefulEvent _creativestormshieldgamephasestepupdatedasynctask))
    (:else)))

(event EventBeginPlay
  (Class|FortAsyncActionCurrentPlaylistReady|PlaylistDataReadyAsync
    (:OnReady
      (if (ScalableFloat|GetValueasBool (Variables|Default|GetRow_Enabled))
        (Networking|SwitchHasAuthority
          (:Authority
            (CallFunction|Initialize)))))))

(event Custom|OnGamePhaseStepChanged
  (params (GamePhaseStep :type EAthenaGamePhaseStep%20Enum))
  (Storm|CheckNewStormCircle GamePhaseStep)
  (if (Utilities|Enum|Equal(Enum) (Variables|SpawnRules|GetSpawnPhaseStep) GamePhaseStep)
    (Variables|Default|SetCurrentNumDropsThisPhaseStep 0)
    (Development|PrintString "Spawn phase step began")
    (Utilities|FlowControl|Delay (ScalableFloat|GetValueatLevel (Variables|Default|GetSpawnDelayTimePerPhaseStep) (Math|Conversions|ToFloat(Integer) (Variables|Storm|GetCurrentStormCircleNum))))
    (Spawning|TrySpawn GamePhaseStep)))

(event Custom|SpawnSupplyDrop
  (Development|PrintString "Spawning supply drop")
  (Spawning|SpawnSupplyDropActor (Spawning|GetSupplyDropTargetLocation))
  (Math|Integer|IncrementInt (Variables|Default|GetCurrentNumDropsThisPhaseStep)
    (:%20%20
      (Math|Integer|IncrementInt (Variables|Default|GetCurrentTotalNumDrops)))))

(event Custom|OnSupplyDropSpawned
  (params (SpawnedSupplyDrop :type Athena%20Supply%20Drop%20Object%20Reference)))

(event Custom|DebugSpawnSpecialSupplyDrop
  (params (ID :type String) (X :type Float%20%28double-precision%29) (Y :type Float%20%28double-precision%29) (Z :type Float%20%28double-precision%29))
  (if (Utilities|String|Equal,CaseInsensitive(String) ID (Utilities|String|ToString(Name) (Variables|Default|GetSupplyDropID)))
    (CallFunction|DoGlobalCalloutSound)
    (Spawning|SpawnSupplyDropActor 0 (Math|Vector|MakeVector X Y Z))))

(event Custom|DebugStartSpecialSupplyDropSpawnPhase
  (params (ID :type String))
  (if (Utilities|String|Equal,CaseInsensitive(String) ID (Utilities|String|ToString(Name) (Variables|Default|GetSupplyDropID)))
    (Spawning|TrySpawn "None")))

(event Custom|DoGlobalCalloutSound
  (Game|SpawnActorfromClass (Variables|Default|GetCalloutClass) (Transformation|GetActorTransform) "Undefined" "MultiplyWithRoot" (Fortnite|Athena|GameState|GetGameStateAthena) 0))

(fn CanSpawn
  (params (GamePhaseStep :type EAthenaGamePhaseStep%20Enum) (StormCircleNum :type Integer))
  (bind _returnvalue (Storm|CheckExplicitAllowedStormCircles StormCircleNum))
  (bind _returnvalue_2 (Spawning|CheckSpawnChance StormCircleNum))
  (bind _returnvalue_3 (CallFunction|HasValidDropLocations))
  (bind _returnvalue_4 (CallFunction|CheckNumDrops))
  (bind _returnvalue_5 (Math|Boolean|ANDBoolean _returnvalue _returnvalue_3 _returnvalue_2 _returnvalue_4))
  (Development|PrintString (Utilities|String|SelectString "Can spawn" "Can NOT spawn" _returnvalue_5))
  (return _returnvalue_5))

(fn CanTryToSpawnThisStep
  (params (Step :type EAthenaGamePhaseStep%20Enum))
  (return (Utilities|Enum|Equal(Enum) Step (Variables|SpawnRules|GetSpawnPhaseStep))))

(fn CheckExplicitAllowedStormCircles
  (params (StormCircleNum :type Integer))
  (return (or (<= (Utilities|Array|Length (Variables|SpawnRules|GetExplicitSpawnAllowedStormCircles)) 0) (Utilities|Array|ContainsItem (Variables|SpawnRules|GetExplicitSpawnAllowedStormCircles) StormCircleNum))))

(fn CheckNewStormCircle
  (params (GamePhaseStep :type EAthenaGamePhaseStep%20Enum))
  (if (Utilities|Enum|Equal(Enum) GamePhaseStep (Variables|Storm|GetNewStormCircleStep))
    (bind _result (Math|Integer|IncrementInt (Variables|Storm|GetCurrentStormCircleNum))
      (:%20%20
        (Development|PrintString (Utilities|String|Append "New storm circle. Circle num: " (Utilities|String|ToString(Integer) _result)))
        ;[1 1 1] Pare down our current list of potential drop locations based on which ones are still inside the new safe zone
        (bind _outvalidlocations (Query|GetValidQueryLocationsforCurrentStormCircle))
        (bind _output_get (Variables|Query|SetPotentialDropLocations _outvalidlocations))
        (Development|PrintString (Utilities|String|Append "Num valid locations remaining: " (Utilities|String|ToString(Integer) (Utilities|Array|Length _output_get))))))))

(fn CheckNumDrops
  (bind _totalmaxdrops (Variables|Default|GetTotalMaxDrops))
  (bind _returnvalue (ScalableFloat|GetValueatLevel _totalmaxdrops))
  (bind _currenttotalnumdrops (Variables|Default|GetCurrentTotalNumDrops))
  (bind _returnvalue_2 (< _currenttotalnumdrops _returnvalue))
  (Utilities|FlowControl|Sequence
    (:then_0
      (if (not _returnvalue_2)
        (Development|PrintString "Cannot spawn because total max drops reached")))
    (:then_1
      (return (and (< (Variables|Default|GetCurrentNumDropsThisPhaseStep) (ScalableFloat|GetValueatLevel (Variables|Default|GetMaxDropsPerPhaseStep) (Math|Conversions|ToFloat(Integer) (Variables|Storm|GetCurrentStormCircleNum)))) _returnvalue_2)))))

(fn CheckSpawnChance
  (params (StormCircleNum :type Integer))
  (bind _returnvalue (Math|Conversions|ToFloat(Integer) StormCircleNum))
  (bind _row_spawnchancebycircle (Variables|SpawnRules|GetRow_SpawnChanceByCircle))
  (bind _returnvalue_2 (ScalableFloat|GetValueatLevel _row_spawnchancebycircle _returnvalue))
  (bind _returnvalue_3 (Math|Random|RandomFloatinRange 0.0 1.0))
  (bind _returnvalue_4 (>= _returnvalue_2 _returnvalue_3))
  (Development|PrintString (Utilities|String|Append "Spawn chance: " (Utilities|String|ToString(Float) _returnvalue_2) " Dice roll: " (Utilities|String|ToString(Float) _returnvalue_3) " Chance succeeded: " (Utilities|String|ToString(Boolean) _returnvalue_4)))
  (return _returnvalue_4))

(fn DoEQSQueryForSpawnLocations
  (bind _returnvalue (AI|EQS|RunEQSQuery (Variables|Query|GetDropLocationQuery) self "AllMatching"))
  (AI|EQS|BindEventtoOnQueryFinished _returnvalue (EventDispatchers|CreateEvent :!event OnDropLocationQueryFinished)))

(fn GetNumDropAttempts
  (return (Math|Random|RandomIntegerinRange (Math|Float|Truncate (ScalableFloat|GetValueatLevel (Variables|Default|GetMinDropsPerPhaseStep) (Math|Conversions|ToFloat(Integer) (Variables|Storm|GetCurrentStormCircleNum)))) (Math|Float|Truncate (ScalableFloat|GetValueatLevel (Variables|Default|GetMaxDropsPerPhaseStep) (Math|Conversions|ToFloat(Integer) (Variables|Storm|GetCurrentStormCircleNum)))))))

(fn GetRandomLocationIndex
  (return (Math|Random|RandomIntegerinRange 0 (- (Utilities|Array|Length (Variables|Query|GetPotentialDropLocations)) 1))))

(fn GetSupplyDropClass
  (return (Variables|Default|GetSupplyDropClass)))

(fn GetSupplyDropTargetLocation
  (if (Variables|SpawnRules|GetbPickRandomQueryLocation)
    (LocalVariables|SetLocationIndex (Utility|GetRandomLocationIndex)))
  (LocalVariables|SetOutLocation (Utilities|Array|Get(acopy) (Variables|Query|GetPotentialDropLocations) (LocalVariables|GetLocationIndex)))
  (Utilities|Array|RemoveIndex (Variables|Query|GetPotentialDropLocations) (LocalVariables|GetLocationIndex))
  (return (LocalVariables|GetOutLocation)))

(fn GetValidQueryLocationsForCurrentStormCircle
  (if (> (Utilities|Array|Length (Variables|Query|GetPotentialDropLocations)) 0)
    (for _array_element (Variables|Query|GetPotentialDropLocations)
      (bind _bisvalid (Query|IsQueryLocationValid _array_element))
      (if _bisvalid
        (Utilities|Array|Add (LocalVariables|GetValidLocations) _array_element)))
    (return (LocalVariables|GetValidLocations))
    (else
      (Development|PrintString "Queried locations array is empty!" true true "(R=1.000000,G=0.000000,B=0.021100,A=1.000000)"))))

(fn HasValidDropLocations
  (bind _potentialdroplocations (Variables|Query|GetPotentialDropLocations))
  (bind _returnvalue (Utilities|Array|Length _potentialdroplocations))
  (bind _returnvalue_2 (> _returnvalue 0))
  (Development|PrintString (Utilities|String|SelectString "Has valid spawn locations" "Does not have valid spawn locations" _returnvalue_2))
  (return _returnvalue_2))

(fn IsQueryLocationValid
  (params (QueryLocation :type Vector))
  (bind _returnvalue (SafeZone|IsLocationinSafeZone QueryLocation))
  (return _returnvalue))

(fn OnDropLocationQueryFinished
  (params (QueryInstance :type EQS%20Query%20Instance%20Object%20Reference) (QueryStatus :type EEnvQueryStatus%20Enum :default "Processing"))
  (if (Utilities|Enum|Equal(Enum) QueryStatus "Success")
    (bind (_resultlocations _returnvalue) (AI|EQS|GetQueryResultsasLocations (Variables|GetQueryInstance)))
    (Development|PrintString (Utilities|String|ToString(Integer) (Utilities|Array|Length _resultlocations)))
    (if _returnvalue
      (Variables|Query|SetQuerySucceeded true)
      (Variables|Query|SetPotentialDropLocations _resultlocations))))

(fn SpawnSupplyDropActor
  (params (DropLocation :type Vector))
  (bind _returnvalue (Game|SpawnActorfromClass (CallFunction|GetSupplyDropClass) (Math|Conversions|ToTransform(Vector) (+ DropLocation (Variables|SpawnRules|GetSpawnPositionOffset))) "Undefined" "SelectDefaultAtRuntime" 0 0 0))
  (CallFunction|OnSupplyDropSpawned _returnvalue)
  (return _returnvalue))

(fn TrySpawn
  (params (GamePhaseStep :type EAthenaGamePhaseStep%20Enum))
  ;[1 1 1] If evaluating chance for every spawn attempt, don't check the chance now
  (if (Variables|Default|GetbEvaluateChancePerAttempt)
    _
    (elif (Spawning|CanSpawn (Variables|GetGamePhaseStep) (Variables|Storm|GetCurrentStormCircleNum))
      (CallFunction|DoGlobalCalloutSound)))
  (for _ (range (CallFunction|GetNumDropAttempts))
    ;[1 1 1] If we are not evaluating chance per attempt, if we've gotten to this point that means we've already succeeded so we can skip the check. Otherwise, check the chance now
    (if (Variables|Default|GetbEvaluateChancePerAttempt)
      (if (Spawning|CanSpawn (Variables|GetGamePhaseStep) (Variables|Storm|GetCurrentStormCircleNum))
        _))
    (CallFunction|SpawnSupplyDrop)))

(fn ConstructionScript)
